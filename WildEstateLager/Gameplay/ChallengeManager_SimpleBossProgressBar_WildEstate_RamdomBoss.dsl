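; Wild Estate random-boss challenge manager with a boss progress bar, extending a parent manager and self-destructing at invalid locations.
; BeginPlay caches playlist team size and removes the placed TroopFactor TV, saving its class and transform.
; Authority init moves to the boss tag's point provider, scales table health and shield by team size, and respawns the TV manager for its tag, otherwise running parent init.
; Final-fight progress tracks combined health and shield.
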(event Custom|CheckDebugActivation
  (if (and (Variables|Default|GetDebug) (!= (GameplayTags|GetTagName (Variables|Boss|GetBossIDTag)) (GameplayTags|GetTagName (GameplayTags|MakeLiteralGameplayTag "(TagName=\"AISpawnerData.Type.Boss\")"))))
    (CallFunction|Init)))

(event EventBeginPlay
  (bind _troopfactortv (Variables|Default|GetTroopFactorTV))
  (if (ScalableFloat|GetValueasBool (Variables|Default|GetIsValidLocation))
    (bind (_gamestate _playlist) (Class|FortAsyncActionCurrentPlaylistReady|PlaylistDataReadyAsync)
      (:OnReady
        (Variables|Default|SetPlaylistMaxTeamSize (Math|Integer|Clamp(Integer) (Variables|/Script/FortniteGame.FortPlaylist|GetMaxTeamSize _playlist) 1 4))
        (Variables|Default|SetTroopFactorTV_Transform (Transformation|GetActorTransform _troopfactortv))
        (Variables|Default|SetTroopFactorTV_Class (Utilities|GetClass _troopfactortv))
        (Gameplay|SilentDie _troopfactortv true :!class /Script/FortniteGame.BuildingActor)
        (CallFunction|CheckDebugActivation)))
    (else
      (Actor|DestroyActor))))

(event Collision|EventActorBeginOverlap
  (params (OtherActor :type Actor%20Object%20Reference)))

(event EventTick
  (params (DeltaSeconds :type Float%20%28single-precision%29)))

(event EventAISpawned
  (params (SpawnedActor :type Actor%20Object%20Reference) (bResult :type Boolean))
  (|CallParentFunction SpawnedActor bResult)
  (Progress|HandleBossHealthChanged))

(event Custom|SpawnedBoss_OnDamaged
  (params (DamagedActor :type Actor%20Object%20Reference) (Damage :type Float%20%28single-precision%29) (InstigatedBy :type Controller%20Object%20Reference) (DamageCauser :type Actor%20Object%20Reference) (HitLocation :type Vector) (FHitComponent :type Primitive%20Component%20Object%20Reference) (BoneName :type Name) (Momentum :type Vector))
  (bind _asfort_player_pawn_athena (Utilities|Casting|CastToFortPlayerPawnAthena (Variables|Boss|GetBossActor)))
  (EventDispatchers|UnbindEventfromOnDamaged _asfort_player_pawn_athena (delegate Custom|SpawnedBoss_OnDamaged))
  (CallFunction|StartSimpleChallenge))

(event EventInit
  (bind _playlistmaxteamsize (Variables|Default|GetPlaylistMaxTeamSize))
  (bind _returnvalue (- _playlistmaxteamsize 1))
  (bind _returnvalue_2 (Math|Conversions|ToFloat(Integer) _returnvalue))
  (bind _bossinfotable (Variables|Default|GetBossInfoTable))
  (Networking|SwitchHasAuthority
    (:Authority
      (CallFunction|SetBossPointProvider)
      (bind _outrownames (DataTable|GetDataTableRowNames _bossinfotable))
      (bind _array_element (Utilities|Array|ForEachLoopwithBreak _outrownames)
        (:LoopBody
          (bind _returnvalue_3 (Utilities|GetDataTableRow _bossinfotable _array_element))
          (if (== (Variables|Boss|GetBossIDTag) (Utilities|Struct|BreakWildEstateBossInfoStruct _returnvalue_3))
            (Variables|Default|SetMaxHealth (ScalableFloat|GetValueatLevel (.MaxHealth_6_97EA1B054CAE72E11E23679E1B34FCFA (Utilities|Struct|BreakWildEstateBossInfoStruct _returnvalue_3)) _returnvalue_2))
            (Variables|Default|SetMaxShield (ScalableFloat|GetValueatLevel (.MaxShield_8_87C7126F4FCC59FB9512A98666620D7F (Utilities|Struct|BreakWildEstateBossInfoStruct _returnvalue_3)) _returnvalue_2))))
        (:Completed
          _))
      (if (== (Variables|Boss|GetBossIDTag) (Variables|Default|GetTroopFactorTV_Tag))
        (CallFunction|PrepareTroopFactorSpawn)
        (else
          (|CallParentFunction)))
      (if (Variables|Default|GetDebug)
        (Development|PrintString (Utilities|String|Append "ChallengeManager: " (GameplayTags|GetDebugStringfromGameplayTag (Variables|Default|GetPOIFactionTag)) "; " (GameplayTags|GetDebugStringfromGameplayTag (Variables|Boss|GetBossIDTag))) true true "(R=0.000000,G=0.660000,B=1.000000,A=1.000000)" 2e+01)))))

(event Custom|SetBossPointProvider
  (bind _bossidtag (Variables|Boss|GetBossIDTag))
  (bind _bosstagtopathprovider (Variables|Default|GetBossTagToPathProvider))
  (bind _value (Utilities|Map|Find _bosstagtopathprovider _bossidtag))
  (Utilities|IsValid _value
    (:Is%20Valid
      (bind _output_get (Variables|References|SetPatrolPointProvider _value))
      (Transformation|SetActorLocation (Class|FortAthenaLivingWorldPointProviderInterface|GetPointProviderLocation _output_get) true))))

(event Custom|PrepareTroopFactorSpawn
  (bind _patrolpointprovider (Variables|References|GetPatrolPointProvider))
  (bind _returnvalue (Game|SpawnBuildingActorfromClass (Variables|Default|GetTroopFactorTV_Class) (Variables|Default|GetTroopFactorTV_Transform) 0 false 0 0))
  (bind _asbp_tvscare_troop_factor_manager (Utilities|Casting|CastToBP_TVScare_TroopFactor_Manager _returnvalue))
  (Variables|/FNM_TVScare/Blueprints/BP_TVScare_Manager.BP_TVScare_Manager_C|SetPatrolPointProvider (Variables|References|GetPatrolPointProvider) _asbp_tvscare_troop_factor_manager)
  (Class|BPTVScareTroopFactorManager|InitforChallengeManager _asbp_tvscare_troop_factor_manager)
  (Utilities|FlowControl|Sequence
    (:then_0
      (Progress|CalculatePhaseProgressTargets)
      (Progress|CalculateTotalProgressNeededPerStage)
      (Utilities|IsValid _patrolpointprovider
        (:Is%20Valid
          (EventDispatchers|BindEventtoOnActorSpawnedResult _patrolpointprovider (EventDispatchers|CreateEvent :!event AISpawned)))))
    (:then_1
      (CallFunction|UpdateAreaExitCollision))))

(fn HandleBossHealthChanged
  (bind _bossactor (Variables|Boss|GetBossActor))
  (if (Variables|Default|GetBossIsReadyForFinalFight)
    (Utilities|IsValid _bossactor
      (:Is%20Valid
        (bind _asbp_player_pawn_non_participant_boss (Utilities|Casting|CastToBP_PlayerPawn_NonParticipant_Boss _bossactor))
        (Variables|Progress|SetBossProgress (Utilities|NormalizeFloat (+ (* (FortPawn|HealthPercent _asbp_player_pawn_non_participant_boss) (Variables|Default|GetMaxHealth)) (* (FortPawn|ShieldPercent _asbp_player_pawn_non_participant_boss) (Variables|Default|GetMaxShield))) 0 (+ (Variables|Default|GetMaxHealth) (Variables|Default|GetMaxShield)) true))))))

(fn ConstructionScript
  (|CallParentFunction))
